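; Fake CoolCarpet actor for the Sneaky Snowman consumable, owned by an Athena_Player_SneakySnowmanV2.
; It binds to its own damage event and destroys that owning snowman when the instigator is not the controller of the snowman's player pawn.
; ShouldDie always returns false, so damage acts through that handler.
; CloseFlaps and OpenFlaps play and reverse one timeline that mirrors rotation on two flap meshes, snapping rotations on finish.
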(event EventBeginPlay
  (|CallParentFunction)
  (EventDispatchers|BindEventtoOnDamaged (delegate Custom|OnDamaged_)))

(event Custom|OnDamaged_
  (params (DamagedActor :type Actor%20Object%20Reference) (Damage :type Float%20%28single-precision%29) (InstigatedBy :type Controller%20Object%20Reference) (DamageCauser :type Actor%20Object%20Reference) (HitLocation :type Vector) (FHitComponent :type Primitive%20Component%20Object%20Reference) (BoneName :type Name) (Momentum :type Vector))
  (if (Utilities|NotEqual(Object) InstigatedBy (Pawn|GetController (Variables|/Game/Athena/Items/Consumables/SneakySnowman/Version2/Athena_Player_SneakySnowmanV2.Athena_Player_SneakySnowmanV2_C|GetPlayerPawn (Variables|Default|GetMyOwner))))
    (Class|AthenaPlayerSneakySnowmanV2|Destroy (Variables|Default|GetMyOwner))))

(event Collision|EventActorBeginOverlap
  (params (OtherActor :type Actor%20Object%20Reference)))

(event EventTick
  (params (DeltaSeconds :type Float%20%28single-precision%29)))

(event Custom|CloseFlaps
  (bind (_direction _rotation) (Timeline|Flaps Play)
    (:Update
      (Utilities|FlowControl|Sequence
        (:then_0
          (Transformation|SetRelativeRotation (Variables|Default|GetSM_CoolCarpet_Flap1) (Math|Rotator|MakeRotator 0.0 (* (- 1.0 _rotation) -2.2e+02)) false false))
        (:then_1
          (Transformation|SetRelativeRotation (Variables|Default|GetSM_CoolCarpet_Flap2) (Math|Rotator|MakeRotator 0.0 (* (- 1.0 _rotation) 2.2e+02)) false false))))
    (:Finished
      (switch Utilities|FlowControl|Switch|SwitchonETimelineDirection _direction
        (:Forward
          (Transformation|SetRelativeRotation (Variables|Default|GetSM_CoolCarpet_Flap2) (0.0 0.0 0.0) false false))
        (:Backward
          (Transformation|SetRelativeRotation (Variables|Default|GetSM_CoolCarpet_Flap1) (-2.2e+02 0.0 0.0) false false)
          (Transformation|SetRelativeRotation (Variables|Default|GetSM_CoolCarpet_Flap2) (2.2e+02 0.0 0.0) false false))))))

(event Custom|OpenFlaps
  (Timeline|Flaps Reverse))

(fn ShouldDie
  (params (Damage :type Float%20%28single-precision%29) (EventInstigator :type Controller%20Object%20Reference) (DamageCauser :type Actor%20Object%20Reference) (bIsTest :type Boolean) (EffectContext :type Gameplay%20Effect%20Context%20Handle%20Structure%20%28by%20ref%29))
  (return false))

(fn ConstructionScript
  (|CallParentFunction))
