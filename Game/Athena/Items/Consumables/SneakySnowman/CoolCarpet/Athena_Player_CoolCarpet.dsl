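; Player-attached actor for Cool Carpet, a Sneaky Snowman-derived box disguise with animated flaps, movement sounds and material variants.
; Aiming, firing or weapon windup drives a pop-out state machine that reveals the player, then re-hides them.
; The owning client predicts each state and sends it through the server to others.
; States toggle the fake actor, crouch, holstering, camera override, collision and gameplay effects hiding the player from AI and turrets.
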(event Custom|ReloadPressed_Client
  (params (Weapon :type Fort%20Weapon%20Object%20Reference))
  (CallFunction|ReloadPressedServer)
  (Utilities|SetHolster false))

(event Custom|ReloadPressed_Server
  (Utilities|SetHolster false))

(event EventBeginPlay
  (bind _playerpawn (Variables|StoredActors|GetPlayerPawn))
  (bind _currentweapon (Variables|/Script/FortniteGame.FortPawn|GetCurrentWeapon _playerpawn))
  (|CallParentFunction)
  (CallFunction|BindToMovementEvents)
  (Utilities|FlowControl|Sequence
    (:then_0
      (Networking|SwitchHasAuthority
        (:Remote
          (FortPawn|BindEventtoOnWeaponEquippedDelegate (Variables|StoredActors|GetPlayerPawn) (delegate Custom|CustomEvent))
          (EventDispatchers|BindEventtoOnLocalTargetingChanged _currentweapon (EventDispatchers|CreateEvent :!event Custom|ADS_Client))
          (EventDispatchers|AssignOnPressTrigger _currentweapon (EventDispatchers|CreateEvent :!event Custom|FirePressed_Client))
          (EventDispatchers|AssignOnReleaseTrigger _currentweapon (EventDispatchers|CreateEvent :!event Custom|FireReleased_Client))
          (EventDispatchers|BindEventtoOnTryReload _currentweapon (EventDispatchers|CreateEvent :!event Custom|ReloadPressed_Client)))))
    (:then_1
      (Utilities|FlowControl|Delay 1.0)
      (if (and (not (Variables|Default|GetFlapsClosed)) (not (Variables|StateTrackers|GetIsPoppingOut)))
        (CallFunction|CloseFlaps)))
    (:then_2
      (Default|BindEventtoOnJumpOut (delegate Custom|JumpedOut))
      (Game|BindEventtoOnDestroyed (delegate Custom|CleanUp))
      (Default|BindEventtoOnSpawnProp (delegate Custom|SetMatInt))
      (bind _asathena_fake_cool_carpet (Utilities|Casting|CastToAthena_Fake_CoolCarpet (Variables|StoredActors|GetFakeActor)))
      (Variables|Default|SetCastedFakeActor _asathena_fake_cool_carpet))
    (:then_3
      (Class|FortAthenaShowHolsteredWeaponsComponent|BindtoFortPawn (Variables|Default|GetFortAthenaShowHolsteredWeapons) (Variables|StoredActors|GetPlayerPawn)))))

(event EventToggleFakeSnowman
  (params (ToggleFakeOn :type Boolean) (HidePlayerIfTrue :type Boolean) (CrouchedLoc :type Vector :default "0, 0, 0") (RotateToGround :type Boolean) (SetPoppingOut :type Boolean) (CanReplaceWithProp :type Boolean))
  (|CallParentFunction ToggleFakeOn HidePlayerIfTrue CrouchedLoc RotateToGround SetPoppingOut CanReplaceWithProp)
  (if (and ToggleFakeOn CanReplaceWithProp)
    (Variables|Default|SetReplacedByPropTimestamp (GameState|GetServerWorldTimeSeconds (Fortnite|Athena|GameState|GetGameStateAthena)))))

(event Custom|CustomEvent
  (params (NewWeapon :type Fort%20Weapon%20Object%20Reference) (PrevWeapon :type Fort%20Weapon%20Object%20Reference))
  (bind _outputdelegate (EventDispatchers|CreateEvent :!event Custom|ADS_Client))
  (Utilities|FlowControl|Sequence
    (:then_0
      (Utilities|IsValid PrevWeapon
        (:Is%20Valid
          (EventDispatchers|UnbindEventfromOnLocalTargetingChanged PrevWeapon _outputdelegate)
          (EventDispatchers|UnbindEventfromOnPressTrigger PrevWeapon (EventDispatchers|CreateEvent :!event Custom|FirePressed_Client))
          (EventDispatchers|UnbindEventfromOnReleaseTrigger PrevWeapon (EventDispatchers|CreateEvent :!event Custom|FireReleased_Client))
          (EventDispatchers|UnbindEventfromOnTryReload PrevWeapon (EventDispatchers|CreateEvent :!event Custom|ReloadPressed_Client)))))
    (:then_1
      (EventDispatchers|BindEventtoOnLocalTargetingChanged NewWeapon _outputdelegate)
      (EventDispatchers|AssignOnPressTrigger NewWeapon (EventDispatchers|CreateEvent :!event Custom|FirePressed_Client))
      (EventDispatchers|AssignOnReleaseTrigger NewWeapon (EventDispatchers|CreateEvent :!event Custom|FireReleased_Client))
      (EventDispatchers|BindEventtoOnTryReload NewWeapon (EventDispatchers|CreateEvent :!event Custom|ReloadPressed_Client)))))

(event EventEndPlay
  (params (EndPlayReason :type EEndPlayReason%20Enum :default "Destroyed"))
  (|CallParentFunction EndPlayReason)
  (CallFunction|UnbindFromMovementEvents))

(event Custom|HideCustomPartsVFX
  (params (NewParam :type Boolean))
  (bind _playerpawn (Variables|StoredActors|GetPlayerPawn))
  (bind _enumvalue (Utilities|Enum|ForEachEFortCustomPartType false)
    (:LoopBody
      (Utilities|IsValid (FortPawn|GetModifierforPartType _playerpawn _enumvalue)
        (:Is%20Valid
          (Utilities|IsValid (Class|CustomCharacterPartModifier|GetIdleVFXComponent (FortPawn|GetModifierforPartType _playerpawn _enumvalue))
            (:Is%20Valid
              (Rendering|SetVisibility (Class|CustomCharacterPartModifier|GetIdleVFXComponent (FortPawn|GetModifierforPartType _playerpawn _enumvalue)) (not NewParam) true))))))))

(event Custom|CloseFlaps
  ;[1 1 1] Flaps
  (if (not (and (Variables|Default|GetAlreadyJumpedOut) (Utilities|IsValid (Variables|Default|GetCastedFakeActor) :!class /Script/Engine.KismetSystemLibrary)))
    (CallFunction|HideCustomPartsVFX true)
    (Variables|Default|SetFlapsClosed true)
    (Class|AthenaFakeCoolCarpet|CloseFlaps (Variables|Default|GetCastedFakeActor))
    (Audio|SpawnSoundAttached (Variables|Sound|GetCloseSound) (Variables|Transformation|GetRootComponent))
    (Utilities|FlowControl|Sequence
      (:then_0
        (bind (_direction _rotation) (Timeline|Timeline_1 SetNewTime :NewTime (Components|Timeline|GetPlaybackPosition (Variables|/Game/Athena/Items/Consumables/SneakySnowman/CoolCarpet/Athena_Fake_CoolCarpet.Athena_Fake_CoolCarpet_C|GetFlaps (Variables|Default|GetCastedFakeActor))))
          (:Update
            (Utilities|FlowControl|Sequence
              (:then_0
                (Transformation|SetRelativeRotation (Variables|Default|GetSM_CoolCarpet_Flap1) (Math|Rotator|MakeRotator 0.0 (* (- 1.0 _rotation) -2.2e+02)) false false))
              (:then_1
                (Transformation|SetRelativeRotation (Variables|Default|GetSM_CoolCarpet_Flap2) (Math|Rotator|MakeRotator 0.0 (* (- 1.0 _rotation) 2.2e+02)) false false))))))
      (:then_1
        (Timeline|Timeline_1 Play)))))

(event Custom|JumpedOut
  (Variables|Default|SetAlreadyJumpedOut true)
  (CallFunction|OpenFlaps))

(event Custom|SetMatInt
  (params (Prop :type Actor%20Object%20Reference))
  (Networking|SwitchHasAuthority
    (:Authority
      (bind _asathena_prop_cool_carpet (Utilities|Casting|CastToAthena_Prop_CoolCarpet Prop))
      (Variables|/Game/Athena/Items/Consumables/SneakySnowman/CoolCarpet/Athena_Prop_CoolCarpet.Athena_Prop_CoolCarpet_C|SetMatIntOverride (Variables|Default|GetMatInt) _asathena_prop_cool_carpet))))

(event Custom|CleanUp
  (params (DestroyedActor :type Actor%20Object%20Reference))
  (Utilities|SetHolster false
    (:OutputPin
      (GameplayEffects|RemoveActiveEffectswithGrantedTags (Ability|GetAbilitySystemComponent (Variables|StoredActors|GetPlayerPawn)) "(GameplayTags=((TagName=\"Gameplay.Action.Player.PreventADS\")))")
      (CallFunction|HideCustomPartsVFX false))))

(event Custom|OpenFlaps
  ;[1 1 1] Flaps
  (CallFunction|HideCustomPartsVFX false)
  (Class|AthenaFakeCoolCarpet|OpenFlaps (Variables|Default|GetCastedFakeActor))
  (Variables|Default|SetFlapsClosed false)
  (Effects|Components|ParticleSystem|SpawnFXAtLocation (Variables|FX|GetP_PopOut) (Collision|GetActorBounds false))
  (Audio|PlaySoundatLocation (select (Pawn|IsLocallyViewed (Variables|StoredActors|GetPlayerPawn)) (Variables|Sound|GetJumpOpenSound_1P) (Variables|Sound|GetJumpOpenSound_3P)) (Transformation|GetActorLocation))
  (Utilities|FlowControl|Sequence
    (:then_0
      (Timeline|Timeline_1 SetNewTime))
    (:then_1
      (Timeline|Timeline_1 Reverse))))

(event Custom|OnGaveSneakySnowman_1
  (CallFunction|SetFlapsVisibleMulticast))

(event Custom|Set%20Flaps%20Visible%20Multicast
  (Development|SetHiddeninGame (Variables|Default|GetSM_CoolCarpet_Flap1) false false))

(event EventJustAttached
  (|CallParentFunction)
  (CallFunction|SetFlapsVisibleMulticast)
  (CallFunction|CloseFlaps))

(event EventFrequentChecksTimer
  (bind _playerpawn (Variables|StoredActors|GetPlayerPawn))
  (bind _returnvalue (Pawn|GetController _playerpawn))
  (|CallParentFunction)
  (switch Utilities|FlowControl|Switch|SwitchonPopOutStates (Variables|Default|GetPopState)
    (:NewEnumerator0
      (Input|ResetIgnoreMoveInput _returnvalue))
    (:NewEnumerator2
      _)
    (:NewEnumerator3
      _)
    (:NewEnumerator4
      _)
    (:NewEnumerator5
      _)
    (:NewEnumerator6
      _)
    (:NewEnumerator7
      _)
    (:NewEnumerator8
      _))
  (Input|SetIgnoreMoveInput _returnvalue true))

(event Custom|OpenFlaps_Multicast
  (CallFunction|OpenFlaps))

(event Custom|SetCameraOverride
  (params (bOverrideEnabled :type Boolean))
  ;[1 1 1] Pop Out State Machine Management
  (Ability|SendGameplayEventtoActor (Variables|StoredActors|GetPlayerPawn) "(TagName=\"Athena.Event.Gameplay.Coolcarpet.UpdateCameraOverride\")" (Utilities|Struct|MakeGameplayEventData (Math|Float|MakeLiteralFloat (Math|Conversions|ToFloat(Boolean) bOverrideEnabled)))))

(event EventMovementUpdated
  (params (DeltaSeconds :type Float%20%28single-precision%29) (OldLocation :type Vector :default "0, 0, 0") (OldVelocity :type Vector :default "0, 0, 0"))
  (|CallParentFunction DeltaSeconds OldLocation OldVelocity)
  (Utilities|FlowControl|Sequence
    (:then_0
      (if (and (GameplayTags|HasMatchingGameplayTag (Variables|StoredActors|GetPlayerPawn) "(TagName=\"MovementMode.Parachuting\")") (Variables|Default|GetFlapsClosed))
        (CallFunction|OpenFlapsMulticast)))
    (:then_1
      (if (Class|AthenaFunctionLibrary|RowtoBool (Variables|Default|GetPopOutOnWindUp))
        (Networking|SwitchHasAuthority
          (:Authority
            (if (GameplayTags|HasMatchingGameplayTag (Variables|StoredActors|GetPlayerPawn) "(TagName=\"Gameplay.Action.Player.InWindup\")")
              (if (not (Variables|StateTrackers|GetIsPoppingOut))
                (CallFunction|OwningClientADS)))))))))

(event Custom|ADS_Client
  (params (AimDownSights :type Boolean))
  ;[1 1 1] Pop Out State Machine Management
  (if AimDownSights
    (if (Utilities|CanPopOut)
      (CallFunction|ChangePopState (select (Variables|StateTrackers|GetCurrentlyReplacedByFake) "NewEnumerator2" "NewEnumerator8")))
    (else
      (CallFunction|SetCameraOverride true))))

(event Custom|SetPopState_Server
  (params (PopState :type PopOutStates%20Enum :default "NewEnumerator0"))
  ;[1 1 1] Pop Out State Machine Management
  (CallFunction|SetPopStateMulticast PopState)
  (Variables|Default|SetReplacedByPropTimestamp (GameState|GetServerWorldTimeSeconds (Fortnite|Athena|GameState|GetGameStateAthena))))

(event Custom|OwningClient_ADS
  (Variables|Default|SetPoppedFromWindUp true)
  (Variables|Default|SetDidWindUp true)
  (CallFunction|ADSClient true))

(event Custom|FirePressed_Client
  (params (Weapon :type Fort%20Weapon%20Object%20Reference))
  ;[1 1 1] Pop Out State Machine Management
  (Variables|Default|SetCachedFirePress (not (CallFunction|IsWeaponConsumable)))
  (Variables|Default|SetFirePressed true)
  (Variables|Default|SetFiring true)
  (Utilities|Time|ClearandInvalidateTimerbyHandle (Variables|Default|GetStopFiringTimer))
  (if (Utilities|CanPopOut)
    (CallFunction|ChangePopState (select (Variables|StateTrackers|GetCurrentlyReplacedByFake) "NewEnumerator2" "NewEnumerator8"))))

(event Custom|SetPopState_Multicast
  (params (PopState :type PopOutStates%20Enum :default "NewEnumerator0"))
  ;[1 1 1] Pop Out State Machine Management
  (Utilities|IsOwningClient
    (:false
      (bind _output_get (Variables|Default|SetPopState PopState))
      (CallFunction|OnPopStateChanged _output_get))))

(event EventWobble
  (params (Strength :type Float%20%28double-precision%29) (MoveUpAndDown :type Boolean))
  ;[1 1 1] Wobble
  (|CallParentFunction Strength MoveUpAndDown)
  (Rendering|Material|SetScalarParameterValueonMaterials (Variables|Default|GetSM_CoolCarpet_Flap2) "WobbleStartTime" (Utilities|Time|GetGameTimeinSeconds))
  (Rendering|Material|SetScalarParameterValueonMaterials (Variables|Default|GetSM_CoolCarpet_Flap1) "WobbleStrength" Strength))

(event Custom|FireReleased_Client
  (params (Weapon :type Fort%20Weapon%20Object%20Reference))
  ;[1 1 1] Pop Out State Machine Management
  (if (Variables|Default|GetFirePressed)
    (Variables|Default|SetFirePressed false)
    (bind _returnvalue (Utilities|Time|SetTimerbyEvent (delegate Custom|StopFiring) 0.2))
    (Variables|Default|SetStopFiringTimer _returnvalue)))

(event Custom|Server_BlockPopOut
  ;[1 1 1] Pop Out State Machine Management
  (GameplayEffects|ApplyGameplayEffectToTarget :self (Ability|GetAbilitySystemComponent (Variables|StoredActors|GetPlayerPawn)) :GameplayEffectClass "/Game/Athena/Items/Consumables/SneakySnowman/CoolCarpet/GE_BlockADS.GE_BlockADS_C" :Target (Ability|GetAbilitySystemComponent (Variables|StoredActors|GetPlayerPawn)) :Level 0.0))

(event Custom|StopFiring
  ;[1 1 1] Pop Out State Machine Management
  (if (not (Variables|Default|GetFirePressed))
    (Variables|Default|SetFiring false)))

(event Custom|ChangePopState
  (params (PopState :type PopOutStates%20Enum :default "NewEnumerator0"))
  ;[1 1 1] Pop Out State Machine Management
  (Utilities|IsOwningClient
    (:true
      (Networking|SwitchHasAuthority
        (:Remote
          (CallFunction|SetPopStateServer PopState)
          (bind _output_get (Variables|Default|SetPopState PopState))
          (CallFunction|OnPopStateChanged _output_get) ; Happens for owning client here. Rep notify calls it for other clients and server
        )))))

(event Custom|Bind%20to%20Movement%20Events
  ;[1 1 1] Hook into / Unhook out of movement events on the pawn from BeginPlay / EndPlay
  (Utilities|FlowControl|Delay 1.5)
  (if (not (Variables|Sound|GetMovementEventsBound))
    (Variables|Sound|SetMovementEventsBound true)
    (EventDispatchers|BindEventtoOnFootstepEvent (Variables|StoredActors|GetPlayerPawn) (EventDispatchers|CreateEvent :!event OnPlayerFootstep))
    (EventDispatchers|BindEventtoOnPawnLanded (Variables|StoredActors|GetPlayerPawn) (EventDispatchers|CreateEvent :!event OnPlayerLanded))
    (Character|BindEventtoMovementModeChangedDelegate (Variables|StoredActors|GetPlayerPawn) (EventDispatchers|CreateEvent :!event OnPlayerMovementModeChanged))))

(event Custom|WeaponChangeClient
  (CallFunction|SetCameraOverride true)
  (Utilities|FlowControl|Delay 0.1)
  (CallFunction|ADSClient (Utilities|IsClientADS)))

(event EventOnWeaponChange
  (params (NewWeapon :type Fort%20Weapon%20Object%20Reference))
  (|CallParentFunction NewWeapon)
  (CallFunction|WeaponChangeClient))

(event Custom|Unbind%20from%20Movement%20Events
  ;[1 1 1] Hook into / Unhook out of movement events on the pawn from BeginPlay / EndPlay
  (if (Variables|Sound|GetMovementEventsBound)
    (Variables|Sound|SetMovementEventsBound false)
    (EventDispatchers|UnbindEventfromOnFootstepEvent (Variables|StoredActors|GetPlayerPawn) (EventDispatchers|CreateEvent :!event OnPlayerFootstep))
    (EventDispatchers|UnbindEventfromOnPawnLanded (Variables|StoredActors|GetPlayerPawn) (EventDispatchers|CreateEvent :!event OnPlayerLanded))
    (Character|UnbindEventfromMovementModeChangedDelegate (Variables|StoredActors|GetPlayerPawn) (EventDispatchers|CreateEvent :!event OnPlayerMovementModeChanged))))

(event Custom|HandleReloaded
  (if (Utilities|Enum|NotEqual(Enum) (Variables|Default|GetPopState) "NewEnumerator5")
    (Utilities|SetHolster true
      (:OutputPin
        (CallFunction|HandleReloadedClient)))))

(event Custom|HandleReloaded_Client
  (Utilities|SetHolster true))

(event Custom|OnPopStateChanged
  (params (PopState :type PopOutStates%20Enum :default "NewEnumerator0"))
  (bind _playerpawn (Variables|StoredActors|GetPlayerPawn))
  (bind _returnvalue (Ability|GetAbilitySystemComponent _playerpawn))
  (bind _playerpawn_2 (Variables|StoredActors|GetPlayerPawn))
  (bind _returnvalue_2 (Ability|GetAbilitySystemComponent _playerpawn_2))
  (bind _fakeactor (Variables|StoredActors|GetFakeActor))
  (bind _asathena_fake_cool_carpet (Utilities|Casting|CastToAthena_Fake_CoolCarpet _fakeactor))
  (bind _staticmesh (Variables|/Game/Athena/Items/Consumables/SneakySnowman/Version2/Athena_Fake_SNeakySnowmanV2.Athena_Fake_SneakySnowmanV2_C|GetStaticMesh _asathena_fake_cool_carpet))
  (bind _playerpawn_3 (Variables|StoredActors|GetPlayerPawn))
  ;[1 1 1] Popping Out State Machine
  (switch Utilities|FlowControl|Switch|SwitchonPopOutStates PopState
    (:NewEnumerator0
      (Variables|StateTrackers|SetIsPoppingOut false)
      (GameplayEffects|ApplyGameplayEffectToTarget :self _returnvalue :GameplayEffectClass "/Game/Athena/Items/Consumables/SneakySnowman/CoolCarpet/GE_HiddenFromTurrets.GE_HiddenFromTurrets_C" :Target _returnvalue :Level 0.0)
      (GameplayEffects|ApplyGameplayEffectToTarget :self _returnvalue :GameplayEffectClass "/Game/Athena/Items/Consumables/SneakySnowman/CoolCarpet/GE_HiddenFromAI.GE_HiddenFromAI_C" :Target _returnvalue :Level 0.0)
      (GameplayEffects|ApplyGameplayEffectToTarget :self _returnvalue :GameplayEffectClass "/Game/Athena/Items/Consumables/SneakySnowman/Version2/GE_BlockExplosions.GE_BlockExplosions_C" :Target _returnvalue :Level 0.0)
      (Utilities|SetHolster (Utilities|ShouldHolster)
        (:OutputPin
          (GameplayEffects|RemoveActiveEffectswithGrantedTags (Ability|GetAbilitySystemComponent (Variables|StoredActors|GetPlayerPawn)) "(GameplayTags=((TagName=\"Gameplay.Action.Player.PreventADS\")))")
          (Collision|SetActorEnableCollision true))))
    (:NewEnumerator2
      (Utilities|StopMovement)
      (if (Variables|StateTrackers|GetCurrentlyReplacedByFake)
        (CallFunction|RotatetoStraight)
        (Variables|StateTrackers|SetIsPoppingOut true)
        (CallFunction|ChangePopState "NewEnumerator3")
        (else
          (CallFunction|ToggleFakeSnowman true false (CallFunction|GetCrouchedLoc) false true)
          (Variables|StateTrackers|SetIsPoppingOut true)
          (Utilities|StraightenFakeActor)
          (CallFunction|OpenFlaps)
          (Utilities|SetHolster false
            (:OutputPin
              (CallFunction|ChangePopState "NewEnumerator4"))))))
    (:NewEnumerator3
      (bind (_direction _shrinkage _rotation) (Timeline|ShrinkBox PlayFromStart)
        (:Update
          (Networking|SwitchHasAuthority
            (:Remote
              (Transformation|SetRelativeScale3D _staticmesh (Math|Vector|MakeVector 1.0 1.0 (- 1.0 (* _shrinkage 0.2))))
              (Transformation|SetRelativeRotation _staticmesh (Math|Rotator|MakeRotator (* _rotation 8.0)) false false))))
        (:Finished
          (Transformation|SetRelativeScale3D (Variables|/Game/Athena/Items/Consumables/SneakySnowman/Version2/Athena_Fake_SNeakySnowmanV2.Athena_Fake_SneakySnowmanV2_C|GetStaticMesh (Utilities|Casting|CastToAthena_Fake_CoolCarpet (Variables|StoredActors|GetFakeActor))) (1.0 1.0 0.8))
          (Utilities|StraightenFakeActor)
          (CallFunction|ChangePopState "NewEnumerator4"))))
    (:NewEnumerator4
      (Utilities|FlowControl|Sequence
        (:then_0
          (Class|FortAthenaShowHolsteredWeaponsComponent|SetShouldShowHolsteredWeapon (Variables|Default|GetFortAthenaShowHolsteredWeapons) true)
          (GameplayEffects|RemoveActiveEffectswithGrantedTags (Ability|GetAbilitySystemComponent (Variables|StoredActors|GetPlayerPawn)) "(GameplayTags=((TagName=\"Athena.AI.Hidden\"),(TagName=\"Athena.AI.IgnoreAll\")))")
          (Networking|CanExecuteCosmeticEvents
            (:True
              (Effects|Components|ParticleSystem|SpawnFXAtLocation (Variables|FX|GetP_PopOut) (+ (Transformation|GetWorldLocation (Variables|Default|GetSM_SneakySnowman)) "0.000000, 0,15.000000") (Math|Conversions|RotationFromXVector (Transformation|GetForwardVector (Variables|Default|GetSM_SneakySnowman)))))))
        (:then_1
          (if (not (Variables|StateTrackers|GetCurrentlyReplacedByFake))
            (CallFunction|ToggleFakeSnowman true false (CallFunction|GetCrouchedLoc) false true true)))
        (:then_2
          (CallFunction|SetCameraOverride false)
          (Utilities|ShrinkFakeActor)
          (Utilities|StraightenFakeActor)
          (GameplayEffects|RemoveActiveEffectswithGrantedTags _returnvalue_2 "(GameplayTags=((TagName=\"Granted.Athena.EnvItem.Sentry.Turret.Hidden\")))")
          (GameplayEffects|RemoveActiveEffectswithGrantedTags _returnvalue_2 "(GameplayTags=((TagName=\"Gameplay.Action.Player.BlockExplosionDamage\")))")
          (Character|UnCrouch (Utilities|CharOwner))
          (CallFunction|OpenFlaps)
          (CallFunction|SetPlayerHidden false)
          (Input|SetIgnoreMoveInput (Pawn|GetController (Variables|StoredActors|GetPlayerPawn)) true)
          (Collision|SetActorEnableCollision false)
          (Default|CallOnPopOut)
          (Utilities|FlowControl|Delay 0.1)
          (Utilities|SetHolster false
            (:OutputPin
              (Utilities|ShrinkFakeActor)
              (Utilities|FlowControl|Delay 0.2)
              ;;[1 0 0.6736] FORT-273135. Fixes what I suspect is a race condition, where the set actor hidden in game from the Update Movement Tick in the parent is reping down to the client after the delay. (and so the player stays hidden while popped out of box)
              (CallFunction|SetPlayerHidden false)
              (CallFunction|ChangePopState "NewEnumerator5"))))))
    (:NewEnumerator5
      (Utilities|SetHolster false
        (:OutputPin
          (Utilities|StraightenFakeActor)
          (Utilities|IsOwningClient
            (:true
              (Networking|SwitchHasAuthority
                (:Remote
                  (if (or (Math|Boolean|ANDBoolean (Utilities|IsClientADS) (not (Weapon|IsWeaponHolstered (Variables|StoredActors|GetPlayerPawn))) (Utilities|IsStopped)) (Variables|Default|GetCachedFirePress)) :!label _label
                    (if (Variables|Default|GetCachedFirePress)
                      (Utilities|FlowControl|DelayUntilNextTick)
                      (CallFunction|HandleCachedFire))
                    (Utilities|FlowControl|Delay 0.05) ; Wait until player stops ADSing
                    (goto _label)
                    (else
                      (Utilities|IsOwningClient
                        (:true
                          (if (Variables|Default|GetPoppedFromWindUp)
                            (Utilities|SetHolster (Utilities|ShouldHolster)
                              (:OutputPin
                                (Utilities|FlowControl|Delay 0.2) ; Let throw animation play before popping back in
                                (Variables|Default|SetPoppedFromWindUp false))))
                          (CallFunction|ServerBlockPopOut)
                          (Class|FortAthenaShowHolsteredWeaponsComponent|SetShouldShowHolsteredWeapon (Variables|Default|GetFortAthenaShowHolsteredWeapons) false)
                          (Variables|Default|SetPoppedFromStand false)
                          (CallFunction|ChangePopState "NewEnumerator6"))))))))))))
    (:NewEnumerator6
      (CallFunction|SetCameraOverride true)
      (Character|Crouch (Utilities|CharOwner))
      (CallFunction|CloseFlaps)
      (Collision|SetActorEnableCollision true)
      (bind (_direction_2 _growth) (Timeline|Grow%20Box PlayFromStart)
        (:Update
          (Transformation|SetRelativeScale3D (Variables|/Game/Athena/Items/Consumables/SneakySnowman/Version2/Athena_Fake_SNeakySnowmanV2.Athena_Fake_SneakySnowmanV2_C|GetStaticMesh (Utilities|Casting|CastToAthena_Fake_CoolCarpet (Variables|StoredActors|GetFakeActor))) (Math|Vector|MakeVector 1.0 1.0 (+ (* _growth 0.2) 0.8))))
        (:Finished
          (Transformation|SetRelativeScale3D (Variables|/Game/Athena/Items/Consumables/SneakySnowman/Version2/Athena_Fake_SNeakySnowmanV2.Athena_Fake_SneakySnowmanV2_C|GetStaticMesh (Utilities|Casting|CastToAthena_Fake_CoolCarpet (Variables|StoredActors|GetFakeActor))) "1.000000,1.000000,1.000000")
          (Default|CallOnPopIn)
          (Utilities|SetHolster (Utilities|ShouldHolster)
            (:OutputPin
              (CallFunction|ChangePopState "NewEnumerator7"))))))
    (:NewEnumerator7
      (CallFunction|SetPlayerHidden (Variables|StateTrackers|GetCurrentlyReplacedByFake))
      (Utilities|FlowControl|Delay 0.1)
      (if (Utilities|IsStopped)
        (CallFunction|RotatetoGround))
      (CallFunction|ChangePopState "NewEnumerator0"))
    (:NewEnumerator8
      (Character|Crouch _playerpawn_3)
      (Pawn|Input|ConsumeMovementInputVector _playerpawn_3)
      (Utilities|FlowControl|Delay 0.3) ; Give some time to finish crouching and setting up.\nThe box state will be enforced later if for some reason this fails
      (CallFunction|ChangePopState "NewEnumerator3"))))

(event Custom|HandleCachedFire
  (bind _playerpawn (Variables|StoredActors|GetPlayerPawn))
  (if (Variables|Default|GetFirePressed)
    (Weapon|PawnStartFire _playerpawn 0)
    (Variables|Default|SetCachedFirePress false)
    (else
      (Weapon|PawnStartFire _playerpawn 0)
      (Variables|Default|SetCachedFirePress false)
      (Weapon|PawnStopFire _playerpawn 0)
      (CallFunction|FireReleasedClient 0))))

(fn FinalPositionCheck
  (return (and (> (Variables|Default|GetReplacedByPropTimestamp) 0) (> (- (GameState|GetServerWorldTimeSeconds (Game|GetGameState)) (Variables|Default|GetReplacedByPropTimestamp)) 0.5))))

(fn GetMaterialInt
  (return (Variables|Default|GetMatInt)))

(fn IsWeaponConsumable
  (return (GameplayTags|HasTag (GameplayTags|GetOwnedGameplayTags (Data|GetWeaponData (Variables|/Script/FortniteGame.FortPawn|GetCurrentWeapon (Variables|StoredActors|GetPlayerPawn))) :!class /Script/GameplayTags.GameplayTagAssetInterface) "(TagName=\"Item.Consumable\")")))

(fn OnPlayerFootstep
  (params (FootIndex :type Integer))
  (Audio|PlaySoundatLocation (select (Pawn|IsLocallyViewed (Variables|StoredActors|GetPlayerPawn)) (Variables|Sound|GetOnFootstepSound_1P) (Variables|Sound|GetOnFootstepSound_3P)) (Transformation|GetActorLocation)))

(fn OnPlayerLanded
  (params (Hit :type Hit%20Result%20Structure%20%28by%20ref%29))
  (Audio|PlaySoundatLocation (select (Pawn|IsLocallyViewed (Variables|StoredActors|GetPlayerPawn)) (Variables|Sound|GetOnLandedSound_1P) (Variables|Sound|GetOnLandedSound_3P)) (Transformation|GetActorLocation))
  (if (not (Variables|StateTrackers|GetIsPoppingOut))
    (CallFunction|CloseFlaps)))

(fn OnPlayerMovementModeChanged
  (params (Character :type Character%20Object%20Reference) (PrevMovementMode :type EMovementMode%20Enum :default "MOVE_None") (PreviousCustomMode :type Byte))
  (switch Utilities|FlowControl|Switch|SwitchonEMovementMode (Variables|/Script/Engine.CharacterMovementComponent|GetMovementMode (Variables|/Script/Engine.Character|GetCharacterMovement Character))
    (:MOVE_Falling
      (Audio|SpawnSoundAttached (Variables|Sound|GetOnJumpedSound) (Variables|Transformation|GetRootComponent)))))

(fn OnRep_MatInt
  (CallFunction|UpdateBoxMaterial (Variables|Default|GetSM_SneakySnowman)))

(fn OnRep_PopState
  (Utilities|IsOwningClient
    (:false
      (CallFunction|OnPopStateChanged (Variables|Default|GetPopState)))))

(fn OnStartedUsingFake
  (|CallParentFunction)
  (Utilities|FlowControl|DoOnce false
    (:Completed
      (CallFunction|UpdateBoxMaterial (Variables|/Game/Athena/Items/Consumables/SneakySnowman/Version2/Athena_Fake_SNeakySnowmanV2.Athena_Fake_SneakySnowmanV2_C|GetStaticMesh (Utilities|Casting|CastToAthena_Fake_CoolCarpet (Variables|StoredActors|GetFakeActor)))))))

(fn OnStoppedUsingFake
  (|CallParentFunction)
  (CallFunction|CloseFlaps))

(fn ShouldShowFake
  (bind (_shouldtoggle _usefake) (|CallParentFunction))
  (switch Utilities|FlowControl|Switch|SwitchonPopOutStates (Variables|Default|GetPopState)
    (:NewEnumerator0
      _)
    (:NewEnumerator1
      _)
    (:NewEnumerator2
      (return (not (Variables|StateTrackers|GetCurrentlyReplacedByFake)) true))
    (:NewEnumerator3
      (return (not (Variables|StateTrackers|GetCurrentlyReplacedByFake)) true))
    (:NewEnumerator4
      (return (not (Variables|StateTrackers|GetCurrentlyReplacedByFake)) true))
    (:NewEnumerator5
      (return (not (Variables|StateTrackers|GetCurrentlyReplacedByFake)) true))
    (:NewEnumerator6
      (return (not (Variables|StateTrackers|GetCurrentlyReplacedByFake)) true))
    (:NewEnumerator7
      _)
    (:NewEnumerator8
      (return (not (Variables|StateTrackers|GetCurrentlyReplacedByFake)) true)))
  (return _shouldtoggle _usefake))

(fn UpdateBoxMaterial
  (params (TargetMeshComponent :type Mesh%20Component%20Object%20Reference))
  (Rendering|Material|SetMaterial TargetMeshComponent 0 (Utilities|Select :Index (Variables|Default|GetMatInt) :Option%200 "/Game/Athena/Items/Consumables/SneakySnowman/Materials/M_CoolCarpet_01.M_CoolCarpet_01" :Option%201 "/Game/Athena/Items/Consumables/SneakySnowman/Materials/M_CoolCarpet_02.M_CoolCarpet_02" :Option%202 "/Game/Athena/Items/Consumables/SneakySnowman/Materials/M_CoolCarpet_03.M_CoolCarpet_03")))

(fn ConstructionScript
  (|CallParentFunction))
