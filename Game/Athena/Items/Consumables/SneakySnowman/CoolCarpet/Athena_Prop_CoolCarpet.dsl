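; Cool Carpet prop that players can hide in, with two animated flaps and three material variants.
; Interaction is gated by a hide flag and movement-state tags, then multicasts the timeline-driven flap opening with sound and applies an aim-down-sights blocking effect to the pawn.
; BeginPlay starts a material wobble and aligns the prop to the ground via line trace.
; The authority derives the material index from location length, with an optional override.
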(event EventBeginPlay
  (bind _sm_coolcarpet_flap1 (Variables|Default|GetSM_CoolCarpet_Flap1))
  (Utilities|SettheMaterial)
  (Rendering|Material|SetScalarParameterValueonMaterials _sm_coolcarpet_flap1 "WobbleStartTime" (Utilities|Time|GetGameTimeinSeconds))
  (Rendering|Material|SetScalarParameterValueonMaterials _sm_coolcarpet_flap1 "WobbleStrength" 1.0)
  (|CallParentFunction)
  (bind (_outhit _returnvalue) (Collision|LineTraceByChannel (+ (Transformation|GetActorLocation) "0, 0,100.000000") (- (Transformation|GetActorLocation) "0, 0,1000.000000") "TraceTypeQuery10"))
  (if _returnvalue
    (Transformation|SetActorRotation (Math|Rotator|MakeRotfromXZ (Math|Vector|Normalize (Math|Vector|ProjectVectorontoPlane (Transformation|GetActorForwardVector) (.Normal (Collision|BreakHitResult _outhit)))) (.Normal (Collision|BreakHitResult _outhit))) true)))

(event Collision|EventActorBeginOverlap
  (params (OtherActor :type Actor%20Object%20Reference)))

(event EventTick
  (params (DeltaSeconds :type Float%20%28single-precision%29)))

(event Interaction|EventBlueprintOnInteract
  (params (InteractingPawn :type Fort%20Pawn%20Object%20Reference) (InteractionBeingAttempted :type EInteractionBeingAttempted%20Enum :default "FirstInteraction"))
  (|CallParentFunction InteractingPawn InteractionBeingAttempted)
  (CallFunction|OpenFlapsMulticast)
  (GameplayEffects|ApplyGameplayEffectToTarget :self (Ability|GetAbilitySystemComponent InteractingPawn) :GameplayEffectClass "/Game/Athena/Items/Consumables/SneakySnowman/CoolCarpet/GE_BlockADS.GE_BlockADS_C" :Target (Ability|GetAbilitySystemComponent InteractingPawn) :Level 0.0))

(event Custom|OpenFlapsMulticast
  (Audio|SpawnSoundAttached (Variables|FX|GetOpenPanelSound) (Variables|Default|GetSM_CoolCarpet_Flap2))
  (bind (_direction _rotation) (Timeline|OpenFlaps Play)
    (:Update
      (Utilities|FlowControl|Sequence
        (:then_0
          (Transformation|SetRelativeRotation (Variables|Default|GetSM_CoolCarpet_Flap1) (Math|Rotator|MakeRotator 0.0 (* _rotation -2.2e+02)) false false))
        (:then_1
          (Transformation|SetRelativeRotation (Variables|Default|GetSM_CoolCarpet_Flap2) (Math|Rotator|MakeRotator 0.0 (* _rotation 2.2e+02)) false false))))
    (:Finished
      (switch Utilities|FlowControl|Switch|SwitchonETimelineDirection _direction
        (:Forward
          (Transformation|SetRelativeRotation (Variables|Default|GetSM_CoolCarpet_Flap1) (-2.2e+02 0.0 0.0) false false)
          (Transformation|SetRelativeRotation (Variables|Default|GetSM_CoolCarpet_Flap2) (2.2e+02 0.0 0.0) false false))
        (:Backward
          (Transformation|SetRelativeRotation (Variables|Default|GetSM_CoolCarpet_Flap2) (0.0 0.0 0.0) false false))))))

(event Custom|CloseFlaps
  (Timeline|OpenFlaps ReverseFromEnd))

(fn BlueprintCanInteract
  (params (InteractingPawn :type Fort%20Pawn%20Object%20Reference) (InteractionBeingAttempted :type EInteractionBeingAttempted%20Enum :default "FirstInteraction") (InteractionType :type TInteractionType%20Enum :default "IT_NoInteraction"))
  (return (Math|Boolean|ANDBoolean (|CallParentFunction InteractingPawn InteractionBeingAttempted InteractionType) (Class|AthenaFunctionLibrary|RowtoBool (Variables|Default|GetCanHideInProp)) (not (GameplayTags|HasAnyMatchingGameplayTags InteractingPawn (GameplayTags|MakeLiteralGameplayTagContainer "(GameplayTags=((TagName=\"MovementMode.Hover\"),(TagName=\"MovementMode.Flying\"),(TagName=\"MovementMode.Floating\"),(TagName=\"MovementMode.Falling\"),(TagName=\"MovementMode.Ballooning\"),(TagName=\"MovementMode.DBNOCarried\"),(TagName=\"MovementMode.Driving\"),(TagName=\"MovementMode.Parachuting\"),(TagName=\"MovementMode.RemoteControl\"),(TagName=\"MovementMode.Passenger\"),(TagName=\"MovementMode.Skydiving\"),(TagName=\"MovementMode.SurfaceSwimming\"),(TagName=\"MovementMode.Ziplining\"),(TagName=\"Granted.Athena.EnvItem.HidingProp.BlockEntrance\")))"))))))

(fn GetMaterialInt
  (return (Math|Integer|SelectInt (Variables|Default|GetMatIntOverride) (Variables|Default|GetRandMatInt) (>= (Variables|Default|GetMatIntOverride) 0))))

(fn IsTargetableByLinearMover
  (return true))

(fn OnRep_MatIntOverride
  (Utilities|SettheMaterial))

(fn RandomizeBoxMaterial
  (bind _matint (CallFunction|GetMaterialInt))
  (Rendering|Material|SetMaterial (Variables|Debug|GetStaticMeshComponent) 0 (Utilities|Select :Index _matint :Option%200 "/Game/Athena/Items/Consumables/SneakySnowman/Materials/M_CoolCarpet_01.M_CoolCarpet_01" :Option%201 "/Game/Athena/Items/Consumables/SneakySnowman/Materials/M_CoolCarpet_02.M_CoolCarpet_02" :Option%202 "/Game/Athena/Items/Consumables/SneakySnowman/Materials/M_CoolCarpet_03.M_CoolCarpet_03")))

(fn ShouldInvisiblePropsBeAllowed
  (return false))

(fn ConstructionScript
  (|CallParentFunction)
  (Networking|SwitchHasAuthority
    (:Authority
      (if (>= (Variables|Default|GetRandMatInt) 0)
        (Variables|Default|SetRandMatInt (Math|Integer|%25%28Integer%29 (Math|Float|Truncate (Math|Vector|VectorLength (Transformation|GetActorLocation))) 3))))))
