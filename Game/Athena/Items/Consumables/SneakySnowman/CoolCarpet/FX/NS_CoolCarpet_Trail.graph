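# Niagara ribbon trail effect for the CoolCarpet variant of the Sneaky Snowman consumable.
# A single Trail emitter spawns short-lived, downward-drifting ribbon particles through FN_SpawnRibbon_v2, with plain SpawnRate disabled.
# Curves over particle life vary ribbon width, fade alpha and drive dynamic material parameters.
# A second alpha scale reads the Fortnite time-of-day NPC parameter, dimming the trail toward both ends of its range.
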NiagaraSystem /Game/Athena/Items/Consumables/SneakySnowman/CoolCarpet/FX/NS_CoolCarpet_Trail
dependencies {
 usedDataInterfaces [&NiagaraDataInterfaceCurve, &NiagaraDataInterfaceVectorCurve]
 usedDynamicInputs []
 usedModules [&SystemState]
}
emitters {Trail {
  emitterUpdateScript [
   &EmitterState(Life Cycle Mode=::NewEnumerator1, Loop Behavior=::NewEnumerator1, Loop Duration=4, UseLoopDelay=false)
   &SpawnRate(SpawnRate=30, Use Spawn Probability=false) {enabled=false}
   &FN_SpawnRibbon_v2()
  ]
  particleSpawnScript [
   &InitializeParticle(Lifetime=0.6, Ribbon Width=30, Ribbon Width Mode=::NewEnumerator1, UsePositionOffset=false, Write Material Random=true)
   &AddVelocity(Velocity=<0 0 -50>)
  ]
  particleUpdateScript [
   /SetVariables(Particles.ScaledRibbonWidth:NiagaraFloat=&Multiply_Float(
     A=$Particles.RibbonWidth
     B=&FloatFromCurve(FloatCurve=DataInterface() {propertyValues {Curve {keys [<at 0.1 1 RCIM_Cubic>, {arriveTangent=0.5556, interpMode=RCIM_Cubic, leaveTangent=0.5556, time=0.3, value=0.8}, <at 1 1.5 RCIM_Cubic>]}}})
    )) {isSetParameters=true}
   &ParticleState()
   &ScaleColor(
    Scale Alpha=&FloatFromCurve(
     FloatCurve=DataInterface() {propertyValues {Curve {keys [
         {arriveTangent=26.124, interpMode=RCIM_Cubic, time=0, value=0}
         {arriveTangent=2, interpMode=RCIM_Cubic, leaveTangent=2, time=0.1, value=1}
         {arriveTangent=-1.1111, interpMode=RCIM_Cubic, leaveTangent=-1.1111, time=0.5, value=1}
         {interpMode=RCIM_Cubic, leaveTangent=-3.7606, time=1, value=0}
        ]}}}
     Scale Curve=0.05
    )
    Scale RGB=&VectorFromCurve(VectorCurve=DataInterface() {propertyValues {
       XCurve {keys [<at 0 0.8746 RCIM_Cubic>,<at 1 0.3652>]}
       YCurve {keys [<at 0 0.913 RCIM_Cubic>,<at 1 0.4318>]}
       ZCurve {keys [<at 0 1 RCIM_Cubic>,<at 1 0.4583>]}
      }})
    ScaleA=true
    ScaleRGB=false
   )
   &ScaleColor(
    Scale Alpha=&FloatFromCurve(
     FloatCurve=DataInterface() {propertyValues {Curve {keys [
         {interpMode=RCIM_Cubic, tangentMode=RCTM_User, time=0, value=0.3}
         {interpMode=RCIM_Cubic, tangentMode=RCTM_User, time=0.5, value=1}
         {tangentMode=RCTM_User, time=1, value=0.3}
        ]}}}
     CurveIndex=Linked() {linkedVariable {name=NPC.FortniteNPC.TimeOfDay_0-1, type=NiagaraFloat}}
    )
    ScaleA=true
    ScaleRGB=false
   ) {module=ScaleColor001}
   &SolveForcesAndVelocity(Clamp Velocity=false, Limit Acceleration=false)
   &DynamicMaterialParameters(
    Index 0 Param 1=&FloatFromCurve(FloatCurve=DataInterface() {propertyValues {Curve {keys [
         <at 0 0.5 RCIM_Cubic>
         {arriveTangent=0.3095, interpMode=RCIM_Cubic, leaveTangent=0.3095, time=0.1, value=1}
         {arriveTangent=-1.1111, interpMode=RCIM_Cubic, leaveTangent=-1.1111, time=0.4816, value=0.649}
         {interpMode=RCIM_Cubic, leaveTangent=-2.6637, time=1, value=0}
        ]}}})
    Index 0 Param 2=&FloatFromCurve(FloatCurve=DataInterface() {propertyValues {Curve {keys [{arriveTangent=19.7743, interpMode=RCIM_Cubic, time=0, value=0}, <at 0.1 1 RCIM_Cubic>, <at 0.3 0>]}}})
    Index 0 Param 3=-0.2
    Index 0 Param 4=1
    Module.Index0Param0Write=true
    Module.Index0Param1Write=true
    Module.Index0Param2Write=true
    Module.Index0Param3Write=true
   )
  ]
  rendererClasses [&NiagaraRibbonRendererProperties]
  renderers [{rendererClass=&NiagaraRibbonRendererProperties}]
 }}
summary {
 systemName=NS_CoolCarpet_Trail
 emitters [{emitterName=Trail, rendererClasses [&NiagaraRibbonRendererProperties]}]
 userVariables []
}
